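; An EQS context that supplies bombardment target locations for environment queries run by the FirePetal WanderingBoss.
; It overrides ProvideLocationsSet, casts the querier actor to BP_PlayerPawn_FirePetal_WanderingBoss and returns that pawn's Bombardment_TargetCenteredTargetArray.
; The context only reads that array from the pawn, and a failed cast returns a default value instead.
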(fn ProvideLocationsSet
  (params (QuerierObject :type Object%20Reference) (QuerierActor :type Actor%20Object%20Reference))
  (bind _asbp_player_pawn_fire_petal_wandering_boss (Utilities|Casting|CastToBP_PlayerPawn_FirePetal_WanderingBoss QuerierActor)
    (:then
      (return (Variables|/FirePetal_WanderingBoss/NPCs/WanderingBoss/Pawn/BP_PlayerPawn_FirePetal_WanderingBoss.BP_PlayerPawn_FirePetal_WanderingBoss_C|GetBombardment_TargetCenteredTargetArray _asbp_player_pawn_fire_petal_wandering_boss)))
    (:CastFailed
      (return 0))))
